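; BP_S33_Story_Essence_Thrown_Ally is a Season 33 story quest actor, a child Blueprint for what its name indicates is a thrown essence ally.
; Its OverlapSuccessful override runs the parent logic, then hides the static mesh and the quest minimap icon.
; BeginPlay and Tick add no logic, and the construction script only calls the parent.
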(event EventBeginPlay
  (|CallParentFunction) ; This node is disabled and will not be called.\nDrag off pins to build functionality.
)

(event EventTick
  (params (DeltaSeconds :type Float%20%28single-precision%29)))

(fn OverlapSuccessful
  (|CallParentFunction)
  (Rendering|SetVisibility (Variables|Default|GetStaticMesh) false false)
  (MiniMapData|SetVisibility (Variables|Default|GetFortQuestIcon) false))

(fn ConstructionScript
  (|CallParentFunction))
